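# Master material for the MondoInverted vehicle wheel cosmetic, producing animated, tintable rim and tire effects through emissive color on a metallic surface.
# Channel-packed RimTex and TireTex masks, UV-thresholded rotating rim samples and panning gradients feed color layers through named reroutes.
# Vertex red blends the emissive layers, and the LOD2345 static switch selects between rim variants with separate rotator masks for different UV layouts.
# The attributes pass through MF_BaseVehicle_Damage, then MF_VehicleDeRez_Transition.
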Material /VehicleCosmetics/Wheels/MondoInverted/Materials/M_Wheel_MondoInverted
graph {
 properties {
  bAutomaticallySetUsageInEditor=false
  bUseMaterialAttributes=true
  bUsedWithMeshDeformer=true
  bUsedWithSkeletalMesh=true
  blendMode=BLEND_Masked
  pixelDepthOffsetMode=PDOM_Legacy
  usedShadingModels=Default Lit
 }
 nodes [
  n1 = Constant() {r=1}
  n2 = Constant()
  n3 = NamedRerouteUsage()
  n4 = Constant() {r=1}
  n5 = If(A=ComponentMask(None=TextureCoordinate()) {r=true}, "A > B"=Constant(), "A == B"=@n4, "A < B"=@n4) {constB=0.4}
  n6 = NamedRerouteUsage()
  n7 = VectorParameter() {
   group=0_SpecialEffects
   parameterName=RimColorB
   sortPriority=0
   channelNames {a=Emissive, b="", g="", r=""}
   defaultValue=<rgba 1 0.0232 0.0086 1>
  }
  n8 = Multiply(A=NamedRerouteUsage(), B=@n7.RGB)
  n9 = Multiply(A=@n7.RGB, B=NamedRerouteUsage())
  n10 = Multiply(A=Multiply(A=@n7.RGB, B=@n7.Emissive), B=NamedRerouteUsage())
  n11 = VectorParameter() {group=0_SpecialEffects, parameterName=RimColorC, sortPriority=0, channelNames {a=Emissive, b="", g="", r=""}, defaultValue=<rgba 1 0 0 1>}
  n12 = TextureCoordinate()
  n13 = Constant()
  n14 = Multiply(
   A=If(A=ComponentMask(None=@n12) {g=true}, "A > B"=Constant() {r=1}, "A == B"=@n13, "A < B"=@n13) {constB=0.94}
   B=Multiply(A=NamedRerouteUsage(), B=NamedRerouteUsage())
  )
  n15 = TextureCoordinate()
  n16 = Constant() {r=1}
  n17 = Multiply(A=If(A=ComponentMask(None=@n15) {g=true}, "A > B"=Constant(), "A == B"=@n16, "A < B"=@n16) {constB=0.8}, B=NamedRerouteUsage())
  n18 = LinearInterpolate(
   A=Multiply(A=Multiply(A=@n11.RGB, B=@n11.Emissive), B=Add(A=@n17, B=@n14))
   B=Constant3Vector() {constant=<rgba 1 1 1 0>}
   Alpha=Add(A=Multiply(A=@n17, B=@n17), B=Multiply(A=@n14, B=@n14))
  )
  n19 = Add(A=Multiply(A=@n7.RGB, B=Multiply(A=@n5, B=@n6)), B=Add(A=Add(A=@n8, B=@n9), B=Add(A=@n10, B=@n18)))
  n20 = VertexColor()
  n21 = NamedRerouteUsage()
  n22 = NamedRerouteUsage()
  n23 = NamedRerouteUsage()
  n24 = Multiply(A=@n7.RGB, B=@n7.Emissive)
  n25 = NamedRerouteUsage()
  n26 = Multiply(A=NamedRerouteUsage(), B=NamedRerouteUsage())
  n27 = Multiply(A=Multiply(A=@n11.RGB, B=@n11.Emissive), B=@n26)
  n28 = Multiply(A=@n26, B=@n26)
  n29 = Constant3Vector() {constant=<rgba 1 1 1 0>}
  n30 = Add(
   A=Add(A=Multiply(A=@n22, B=@n7.RGB), B=Multiply(A=@n7.RGB, B=@n23))
   B=Add(A=Multiply(A=@n24, B=@n25), B=LinearInterpolate(A=@n27, B=@n29, Alpha=@n28))
  )
  n31 = StaticSwitchParameter(
   True=LinearInterpolate(A=Add(A=@n3, B=@n19), B=NamedRerouteUsage(), Alpha=@n20.R)
   False=LinearInterpolate(A=Add(A=@n21, B=@n30), B=NamedRerouteUsage(), Alpha=@n20.R)
  ) {group=0_SpecialEffects, parameterName=LOD2345}
  MaterialFunctionCall(In=MaterialFunctionCall(Material=MakeMaterialAttributes!(
     BaseColor
     Metallic=@n1
     Specular
     Roughness=@n2
     Anisotropy
     EmissiveColor=@n31
     Opacity
     OpacityMask
     Normal
     Tangent
     WorldPositionOffset
     SubsurfaceColor
     ClearCoat
     ClearCoatRoughness
     AmbientOcclusion
     Refraction
     CustomizedUVs_0
     CustomizedUVs_1
     CustomizedUVs_2
     CustomizedUVs_3
     CustomizedUVs_4
     CustomizedUVs_5
     CustomizedUVs_6
     CustomizedUVs_7
     PixelDepthOffset
     ShadingModel
     Displacement
    )) {materialFunction=&MF_BaseVehicle_Damage}.Result) {materialFunction=&MF_VehicleDeRez_Transition}
  n32 = TextureSampleParameter2D() {
   group=0_SpecialEffects
   parameterName=RimTex
   samplerType=SAMPLERTYPE_LinearColor
   texture=/VehicleCosmetics/Wheels/MondoInverted/Textures/T_Rim_MondoInverted_RGB
  }
  NamedRerouteDeclaration(None=@n32.R) {name=RimRmask, nodeColor=<rgba 1 0.8438 0 1>}
  NamedRerouteDeclaration(None=@n32.G) {name=RimGMask, nodeColor=<rgba 0.5234 0 1 1>}
  n33 = Time()
  n34 = ScalarParameter() {defaultValue=1, parameterName=RotateFXSpeed}
  n35 = TextureSampleParameter2D(UVs=Rotator(Coordinate=TextureCoordinate(), Time=Multiply(A=@n33, B=@n34)) {centerX=0.738, centerY=0.411}) {
   group=0_SpecialEffects
   parameterName=RimTex
   samplerType=SAMPLERTYPE_LinearColor
   texture=&T_Rim_MondoInverted_RGB
  }
  n36 = TextureCoordinate()
  n37 = ComponentMask(None=@n36) {r=true}
  n38 = Constant()
  n39 = If(A=@n37, "A > B"=Constant() {r=1}, "A == B"=@n38, "A < B"=@n38) {constB=0.69}
  n40 = Constant()
  n41 = ComponentMask(None=@n36) {g=true}
  n42 = Constant() {r=1}
  n43 = Constant()
  n44 = Constant()
  n45 = Constant() {r=1}
  n46 = Multiply(
   A=If(A=@n37, "A > B"=@n40, "A == B"=@n40, "A < B"=Constant() {r=1}) {constB=0.78}
   B=Multiply(A=If(A=@n41, "A > B"=@n42, "A == B"=@n42, "A < B"=@n43) {constB=0.36}, B=If(A=@n41, "A > B"=@n44, "A == B"=@n44, "A < B"=@n45) {constB=0.45})
  )
  NamedRerouteDeclaration(None=Multiply(A=@n35.A, B=Multiply(A=@n39, B=@n46))) {name=SmallRotatorLOD01, nodeColor=<rgba 1 0 0.6094 1>}
  n47 = Time()
  n48 = ScalarParameter() {defaultValue=1, parameterName=RotateFXSpeed}
  n49 = TextureSampleParameter2D(UVs=Rotator(Coordinate=TextureCoordinate(), Time=Multiply(A=@n47, B=@n48)) {centerX=0.623, centerY=0.906}) {
   group=0_SpecialEffects
   parameterName=RimTex
   samplerType=SAMPLERTYPE_LinearColor
   texture=&T_Rim_MondoInverted_RGB
  }
  n50 = TextureCoordinate()
  n51 = Constant()
  n52 = If(A=ComponentMask(None=@n50) {r=true}, "A > B"=Constant() {r=1}, "A == B"=@n51, "A < B"=@n51) {constB=0.557}
  n53 = Constant()
  n54 = ComponentMask(None=@n50) {g=true}
  n55 = Constant() {r=1}
  n56 = Constant()
  n57 = ComponentMask(None=@n50) {g=true}
  n58 = Constant()
  n59 = Constant() {r=1}
  n60 = Multiply(
   A=If(A=ComponentMask(None=@n50) {r=true}, "A > B"=@n53, "A == B"=@n53, "A < B"=Constant() {r=1}) {constB=0.69}
   B=Multiply(
    A=If(A=@n54, "A > B"=@n55, "A == B"=@n55, "A < B"=@n56) {constB=0.841}
    B=If(A=@n57, "A > B"=@n58, "A == B"=@n58, "A < B"=@n59) {constB=0.974}
   )
  )
  NamedRerouteDeclaration(None=Multiply(A=@n49.A, B=Multiply(A=@n52, B=@n60))) {name=BigRotatorLOD2, nodeColor=<rgba 1 0.9141 0 1>}
  n61 = VectorParameter() {group=0_SpecialEffects, parameterName=TireColorA, sortPriority=0, channelNames {a=Emissive, b="", g="", r=""}, defaultValue=<rgba 1 0 0 1>}
  n62 = VectorParameter() {group=0_SpecialEffects, parameterName=TireColorB, sortPriority=0, channelNames {a=Emissive, b="", g="", r=""}, defaultValue=<rgba 1 0 0 1>}
  n63 = Multiply(A=@n62.RGB, B=@n62.RGB)
  n64 = NamedRerouteUsage()
  n65 = VectorParameter() {group=0_SpecialEffects, parameterName=TireColorC, sortPriority=0, channelNames {a=Emissive, b="", g="", r=""}, defaultValue=<rgba 1 0 0 1>}
  n66 = Add(A=Multiply(A=Multiply(A=@n63, B=@n64), B=Constant() {r=0.15}), B=Multiply(A=Multiply(A=@n65.RGB, B=@n65.RGB), B=NamedRerouteUsage()))
  n67 = Add(A=Multiply(A=Multiply(A=@n61.RGB, B=@n61.RGB), B=NamedRerouteUsage()), B=@n66)
  n68 = NamedRerouteUsage()
  NamedRerouteDeclaration(None=Add(A=Multiply(A=@n67, B=@n68), B=@n66)) {name=TireFX, nodeColor=<rgba 1 0 0.8203 1>}
  NamedRerouteDeclaration(None=@n32.A) {name=RimAMask, nodeColor=<rgba 0 1 0.3672 1>}
  n69 = VectorParameter() {
   group=0_SpecialEffects
   parameterName=RimColorA
   sortPriority=0
   channelNames {a=Emissive, b="", g="", r=""}
   defaultValue=<rgba 1 0.9823 0.0307 1>
  }
  NamedRerouteDeclaration(None=Multiply(A=Multiply(A=@n69.RGB, B=@n69.RGB), B=NamedRerouteUsage())) {name=ColorA_Output, nodeColor=<rgba 1 0.0469 0 1>}
  NamedRerouteDeclaration(None=@n32.B) {name=RimBMask, nodeColor=<rgba 0 1 0.2031 1>}
  n70 = TextureSampleParameter2D() {
   group=0_SpecialEffects
   parameterName=TireTex
   samplerType=SAMPLERTYPE_LinearColor
   texture=/VehicleCosmetics/Wheels/MondoInverted/Textures/T_Tire_MondoInverted_RGB
  }
  NamedRerouteDeclaration(None=@n70.R) {name=TireRMask, nodeColor=<rgba 1 0 0.1172 1>}
  NamedRerouteDeclaration(None=@n70.G) {name=TireGMask, nodeColor=<rgba 0 0.4375 1 1>}
  NamedRerouteDeclaration(None=@n70.B) {name=TireBMask, nodeColor=<rgba 0.7578 1 0 1>}
  n71 = TextureCoordinate()
  n72 = Multiply(A=Time(), B=ScalarParameter() {defaultValue=1, parameterName=TireFXSpeed})
  NamedRerouteDeclaration(None=TextureSample(UVs=Panner(Coordinate=@n71, Time=@n72) {speedX=-0.5}) {texture=/VehicleCosmetics/Wheels/_SharedTextures/FX/Gradient_D}.R) {name=TirePannerFX, nodeColor=<rgba 1 0.1406 0 1>}
  n73 = TextureCoordinate() {uTiling=30}
  n74 = Multiply(A=Time(), B=ScalarParameter() {defaultValue=2, parameterName=PipeFXSpeed})
  NamedRerouteDeclaration(None=TextureSample(UVs=Panner(Coordinate=@n73, Time=@n74) {speedX=-1}) {texture=&Gradient_D}.R) {name=PipePannerFX, nodeColor=<rgba 1 0.6094 0 1>}
  n75 = Time()
  n76 = ScalarParameter() {defaultValue=1, parameterName=RotateFXSpeed}
  n77 = TextureSampleParameter2D(UVs=Rotator(Coordinate=TextureCoordinate(), Time=Multiply(A=@n75, B=@n76)) {centerX=0.096, centerY=0.468}) {
   group=0_SpecialEffects
   parameterName=RimTex
   samplerType=SAMPLERTYPE_LinearColor
   texture=&T_Rim_MondoInverted_RGB
  }
  n78 = TextureCoordinate()
  n79 = Constant() {r=1}
  n80 = If(A=ComponentMask(None=@n78) {r=true}, "A > B"=Constant(), "A == B"=@n79, "A < B"=@n79) {constB=0.17}
  n81 = ComponentMask(None=@n78) {g=true}
  n82 = Constant() {r=1}
  n83 = Constant()
  n84 = Multiply(
   A=If(A=@n81, "A > B"=@n82, "A == B"=@n82, "A < B"=Constant()) {constB=0.4}
   B=If(A=@n81, "A > B"=@n83, "A == B"=@n83, "A < B"=Constant() {r=1}) {constB=0.55}
  )
  NamedRerouteDeclaration(None=Multiply(A=@n77.A, B=Multiply(A=@n80, B=@n84))) {name=BigRotatorLOD01, nodeColor=<rgba 1 0.9141 0 1>}
  n85 = Time()
  n86 = ScalarParameter() {defaultValue=1, parameterName=RotateFXSpeed}
  n87 = TextureSampleParameter2D(UVs=Rotator(Coordinate=TextureCoordinate(), Time=Multiply(A=@n85, B=@n86)) {centerX=0.484, centerY=0.792}) {
   group=0_SpecialEffects
   parameterName=RimTex
   samplerType=SAMPLERTYPE_LinearColor
   texture=&T_Rim_MondoInverted_RGB
  }
  n88 = TextureCoordinate()
  n89 = Constant()
  n90 = If(A=ComponentMask(None=@n88) {r=true}, "A > B"=Constant() {r=1}, "A == B"=@n89, "A < B"=@n89) {constB=0.44}
  n91 = Constant()
  n92 = ComponentMask(None=@n88) {g=true}
  n93 = Constant() {r=1}
  n94 = Constant()
  n95 = ComponentMask(None=@n88) {g=true}
  n96 = Constant()
  n97 = Constant() {r=1}
  n98 = Multiply(
   A=If(A=ComponentMask(None=@n88) {r=true}, "A > B"=@n91, "A == B"=@n91, "A < B"=Constant() {r=1}) {constB=0.527}
   B=Multiply(
    A=If(A=@n92, "A > B"=@n93, "A == B"=@n93, "A < B"=@n94) {constB=0.748}
    B=If(A=@n95, "A > B"=@n96, "A == B"=@n96, "A < B"=@n97) {constB=0.834}
   )
  )
  NamedRerouteDeclaration(None=Multiply(A=@n87.A, B=Multiply(A=@n90, B=@n98))) {name=SmallRotatorLOD2, nodeColor=<rgba 1 0.9141 0 1>}
 ]
}
